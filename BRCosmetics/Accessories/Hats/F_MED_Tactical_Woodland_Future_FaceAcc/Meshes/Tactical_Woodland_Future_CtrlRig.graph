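# Control Rig for the Tactical Woodland Future face accessory that applies a corrective rotation to the dyn_hair_01 bone as the head moves.
# A spherical pose reader measures head orientation relative to neck_02, with asymmetric active and falloff regions.
# Its output weights a Modify Transforms node that adds a fixed Euler-derived rotation to that bone.
# The other three dyn_hair bones are untouched, and the reader's debug drawing is left enabled.
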ControlRigBlueprint /BRCosmetics/Accessories/Hats/F_MED_Tactical_Woodland_Future_FaceAcc/Meshes/Tactical_Woodland_Future_CtrlRig
graphs [
 RigVMFunctionLibrary {links [], nodes []}
 RigVMModel {nodes [
   RigUnit_BeginExecution = &RigUnit_BeginExecution()
   n1 = &RigUnit_SphericalPoseReader(
    ActiveRegionScaleFactors=(PositiveWidth=0.1,NegativeWidth=2,PositiveHeight=1.5,NegativeHeight=1.5)
    ActiveRegionSize=0.1
    Debug=(bDrawDebug=true,DebugScale=25,DebugSegments=20,DebugThickness=0.25)
    DriverAxis=(X=0,Y=-1,Z=0)
    DriverItem=(Type=Bone,Name="head")
    FalloffRegionScaleFactors=(PositiveWidth=1,NegativeWidth=2,PositiveHeight=2,NegativeHeight=2)
    FalloffSize=0.2
    OptionalParentItem=(Type=Bone,Name="neck_02")
    RotationOffset=(X=-90,Y=0,Z=-10)
    ExecutePin=@RigUnit_BeginExecution.ExecutePin
   )
   > &RigUnit_ModifyTransforms(
    ItemToModify=((Item=(Type=Bone,Name="dyn_hair_01"),Transform=(Rotation=(X=0,Y=0,Z=0,W=1),Translation=(X=0,Y=0,Z=0),Scale3D=(X=1,Y=1,Z=1))))
    ItemToModify.0.Transform.Rotation=&RigVMFunction_MathQuaternionFromEuler(Euler=(X=-10,Y=30,Z=20), RotationOrder=XYZ).Result
    Weight=@n1.OutputParam
   )
  ]}
]
hierarchy [
 root {loc=<0 0 0>, rot=<0 0 0 1>}
 pelvis {parent=root, loc=<0 -0.5301 98.0571>, rot=<0.0127 0.707 -0.0127 -0.707>}
 spine_01 {parent=pelvis, loc=<6.1208 -0.9328 0>, rot=<0 0 0.034 -0.9994>}
 spine_02 {parent=spine_01, loc=<6.5996 -0.2095 0>, rot=<0 0 0.0238 -0.9997>}
 spine_03 {parent=spine_02, loc=<6.4914 -0.1237 0>, rot=<0 0 0.0149 -0.9999>}
 spine_04 {parent=spine_03, loc=<6.9598 0.0794 0>, rot=<0 0 -0.0132 -0.9999>}
 spine_05 {parent=spine_04, loc=<6.9987 0 0>, rot=<0 0 -0.0015 -1>}
 clavicle_l {parent=spine_05, loc=<5.4831 -0.1053 -3.2946>, rot=<0.0465 -0.7289 -0.009 -0.6829>}
 clavicle_r {parent=spine_05, loc=<5.4831 -0.1053 3.2946>, rot=<0.7289 0.0465 -0.6829 0.009>}
 neck_01 {parent=spine_05, loc=<11.3138 0.1529 0>, rot=<0 0 -0.2003 -0.9797>}
 neck_02 {parent=neck_01, loc=<4.5627 0 0>, rot=<0 0 0.0504 -0.9987>}
 head {parent=neck_02, loc=<5.3914 0 0>, rot=<0 0 0.115 -0.9934>}
 faceAttach {parent=head, loc=<-2.2967 -1.5887 0>, rot=<0.0044 -0.3827 -0.0044 -0.9239>, scale=<0.89 0.89 0.89>}
 L_cheek_inner {parent=faceAttach, loc=<4.4315 9.5551 -1.5152>, rot=<0.0454 -0.6262 0.3191 -0.7099>}
 R_cheek_inner {parent=faceAttach, loc=<-1.5797 9.5144 4.4959>, rot=<-0.1979 -0.0545 -0.2598 -0.9436>}
 C_jaw {parent=faceAttach, loc=<1.5885 3.5177 1.6362>, rot=<0.0044 -0.3827 -0.0044 -0.9239>}
 R_eye {parent=faceAttach, loc=<2.0925 8.692 7.0275>, rot=<0.3823 -0.0165 0.9223 -0.0548>}
 L_eye {parent=faceAttach, loc=<6.9687 8.7249 2.1513>, rot=<0.3819 0.0253 0.9227 0.046>}
 dyn_hair_01 {parent=head, loc=<2.2083 -5.5562 2.5139>, rot=<0.0332 -0.9884 0.1482 0.0016>}
 dyn_hair_02 {parent=head, loc=<5.7918 7.4199 -4.8317>, rot=<0.124 -0.8731 0.4107 -0.2314>}
 dyn_hair_03 {parent=head, loc=<6.2876 8.5277 -2.4501>, rot=<-0.0756 -0.9316 0.257 -0.2458>}
 dyn_hair_04 {parent=head, loc=<2.1162 -4.984 -4.0785>, rot=<0 -0.9348 -0.3551 -0.0102>}
]
