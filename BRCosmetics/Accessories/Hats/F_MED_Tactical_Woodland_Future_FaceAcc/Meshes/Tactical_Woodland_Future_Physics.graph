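# Physics asset for the female medium Tactical Woodland Future face accessory, simulating its four dyn_hair bones.
# Box bodies on the dyn_hair bones, plus an extra capsule on dyn_hair_01, are each constrained directly to the head with limited swing and twist.
# Capsules on the head and spine_05 give the hair bones shapes representing the head and upper torso.
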PhysicsAsset /BRCosmetics/Accessories/Hats/F_MED_Tactical_Woodland_Future_FaceAcc/Meshes/Tactical_Woodland_Future_Physics
bodies {
 dyn_hair_01 [
  {shapeType=Capsule, length=5, radius=2, shapeName=dyn_hair_01_capsule, center=<4.2419 1.5668 -5.1753>, rotation=<rot 90 173.2786 173.2786>}
  {shapeType=Box, extentX=8, extentY=2.5762, extentZ=7.7188, center=<4 0 0>}
 ]
 dyn_hair_02 [{shapeType=Box, extentX=4, extentY=1.5, extentZ=2, center=<2 0 0>}]
 dyn_hair_03 [{shapeType=Box, extentX=4, extentY=1.5, extentZ=2, center=<2 0 0>}]
 dyn_hair_04 [{shapeType=Box, extentX=8, extentY=2.5, extentZ=5.5, center=<4 0 0>}]
 head [{shapeType=Capsule, length=11.9555, radius=6.3074, center=<1.6133 1.6045 0>, rotation=<rot 90 0 0>}]
 spine_05 [{shapeType=Capsule, length=15, radius=15, center=<-5.5352 -2.8247 0>}]
}
constraints [
 {bone1Name=dyn_hair_01, bone2Name=head, swing1LimitDegrees=45, swing2LimitDegrees=65, twistLimitDegrees=45}
 {bone1Name=dyn_hair_02, bone2Name=head, swing1LimitDegrees=45, swing2LimitDegrees=45, twistLimitDegrees=45}
 {bone1Name=dyn_hair_04, bone2Name=head, swing1LimitDegrees=45, swing2LimitDegrees=65, twistLimitDegrees=45}
 {bone1Name=dyn_hair_03, bone2Name=head, swing1LimitDegrees=45, swing2LimitDegrees=45, twistLimitDegrees=45}
]
