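# Animation Blueprint for the Tactical Woodland Future face accessory cosmetic, which follows its attached parent mesh and adds dynamic motion.
# It copies the parent's pose, layering the input pose over the four dyn_hair bone branches, then caches the result.
# A RigidBody node simulates that pose relative to spine_03 with scaled component motion and an LOD threshold, then feeds a Control Rig for the final pose.
# The event graph holds no logic.
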AnimBlueprint /BRCosmetics/Accessories/Hats/F_MED_Tactical_Woodland_Future_FaceAcc/Meshes/Tactical_Woodland_Future_Skeleton_AnimBlueprint
target_skeleton=&/Tactical_Woodland_Future_Skeleton
graphs [
 AnimGraph {nodes [
   n1 = RigidBody(ComponentPose=Local To Component(LocalPose=Use cached pose 'ExceptHair'())) {
    bClampLinearTranslationLimitToRefPose=true
    bOverrideWorldGravity=true
    lODThreshold=1
    overridePhysicsAsset=&/Tactical_Woodland_Future_Physics
    simulationSpace=BaseBoneSpace
    alphaBoolBlend {blendOption=Linear}
    baseBoneRef {boneName=spine_03}
    componentAppliedLinearAccClamp=<4000 4000 4000>
    componentLinearAccScale=<0.2 0.2 0.1>
    componentLinearVelScale=<0.2 0.2 0.1>
    overrideWorldGravity {z=-980}
   }
   n2 = Control Rig(Source=Component To Local(ComponentPose=@n1)) {
    alphaBoolBlend {blendOption=Linear}
    controlRigAssetReference {blueprintRigClass=&/Tactical_Woodland_Future_CtrlRig.Tactical_Woodland_Future_CtrlRig_C}
   }
   Output Pose(Result=@n2) {layerGroup=None}
   AnimGraphNode_SubInput_0 = Input Pose()
   n3 = Layered blend per bone(BasePose=Copy Pose From Mesh() {bUseAttachedParent=true}, BlendPoses_0=@AnimGraphNode_SubInput_0) {layerSetup [{branchFilters [
       {blendDepth=1, boneName=dyn_hair_01}
       {blendDepth=1, boneName=dyn_hair_02}
       {blendDepth=1, boneName=dyn_hair_03}
       {blendDepth=1, boneName=dyn_hair_04}
      ]}]}
   Save cached pose 'ExceptHair'(Pose=@n3) {cacheName=ExceptHair}
  ]}
 EventGraph {nodes [
   TryGetPawnOwner()
   Event BlueprintUpdateAnimation()
  ]}
]
variables []
